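; Junk Rift manager for a Fortnite mode that caps how many rifts each player sees per match.
; On the server, it waits for match ready, then seeds a map of per-player rift counts from the GameState player array.
; CheckPlayerRiftCount, noted as called during junk spawn attempts with a player in range, returns true at the cap, otherwise increments the count and returns false.
; Players missing from the map are logged and added.
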(event EventBeginPlay
  (Networking|SwitchHasAuthority
    (:Authority
      (Class|FortAsyncActionAuthorityMatchReady|MatchReadyAsync
        (:OnReady
          (bind _output_get (Variables|Default|SetPlayersInMatch (Variables|/Script/Engine.GameStateBase|GetPlayerArray (Game|GetGameState))))
          (for _array_element _output_get
            (Utilities|Map|Add (Variables|Default|GetPlayerRiftCounts) _array_element))
          (Development|PrintString (Utilities|String|Append "Players in Match map length is: " (Utilities|String|ToString(Integer) (Utilities|Map|Length (Variables|Default|GetPlayerRiftCounts)))) true true "(R=0.000000,G=0.660000,B=1.000000,A=1.000000)" 2e+01))))))

(fn CheckPlayerRiftCount
  (params (PlayerState :type Player%20State%20Object%20Reference))
  (bind _playerriftcounts (Variables|Default|GetPlayerRiftCounts))
  (bind (_value _returnvalue) (Utilities|Map|Find _playerriftcounts PlayerState))
  (bind _currentplayerstate (LocalVariables|GetCurrentPlayerState))
  ;[0 0.15 0.001049 0.5] Called when a Rift has a player in range during Junk Spawn attempt. See if the current in range player has seen the Max Number of rifts in a match.
  (LocalVariables|SetCurrentPlayerState PlayerState)
  (if _returnvalue
    (LocalVariables|SetCurrentRiftCountForPlayer _value)
    (else
      (Fortnite|FortShippingLog self "JRManager: Failed to Find Player State in Map")
      (Utilities|Map|Add _playerriftcounts (LocalVariables|GetCurrentPlayerState))
      (LocalVariables|SetCurrentRiftCountForPlayer 0)))
  (if (>= (LocalVariables|GetCurrentRiftCountForPlayer) (Variables|Default|GetMaxSeenRifts))
    (return true)
    (else
      (Utilities|Map|Remove _playerriftcounts _currentplayerstate)
      (Utilities|Map|Add _playerriftcounts _currentplayerstate (+ (LocalVariables|GetCurrentRiftCountForPlayer) 1))
      (return false))))

(fn ConstructionScript)
