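; BP_GM_TeleportLocations is a placed set of 20 scene-component points that supplies player teleport destinations.
; The construction script gathers the points into an array and copies it to EditableLocationsArray.
; GetUnoccupiedTeleportPoint, noted as called from a teleport Gameplay Ability, returns the world location indexed by PlayersTeleported, then increments it.
; Despite its name, it checks no occupancy; the counter resets upon reaching PlayersPerBucket or through the ResetPlayersTeleported event.
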(event EventBeginPlay)

(event Custom|ResetPlayersTeleported
  (Variables|Default|SetPlayersTeleported 0))

(fn GetUnoccupiedTeleportPoint
  (bind _playersteleported (Variables|Default|GetPlayersTeleported))
  ; Called from Teleport controlling GA.
  (if (== (Variables|Default|GetPlayersTeleported) (Variables|Default|GetPlayersPerBucket))
    (Variables|Default|SetPlayersTeleported 0) ; reset for bucket
    (else
      _))
  (bind _result (ArrayMacros|SafeArrayGet (Variables|Default|GetEditableLocationsArray) _playersteleported)
    (:%20%20
      (Variables|Default|SetSelectedTLocation _result)
      (Math|Integer|IncrementInt _playersteleported
        (:%20%20
          (return (Transformation|GetWorldLocation (Variables|Default|GetSelectedTLocation)) false))))))

(fn ConstructionScript
  (Variables|Default|SetTeleport%20Locations (Utilities|Array|MakeArray (Variables|Default|GetScene%20-%2001) (Variables|Default|GetScene%20-%2002) (Variables|Default|GetScene%20-%2003) (Variables|Default|GetScene%20-%2004) (Variables|Default|GetScene%20-%2005) (Variables|Default|GetScene%20-%2006) (Variables|Default|GetScene%20-%2007) (Variables|Default|GetScene%20-%2008) (Variables|Default|GetScene%20-%2009) (Variables|Default|GetScene%20-%2010) (Variables|Default|GetScene%20-%2011) (Variables|Default|GetScene%20-%2012) (Variables|Default|GetScene%20-%2013) (Variables|Default|GetScene%20-%2014) (Variables|Default|GetScene%20-%2015) (Variables|Default|GetScene%20-%2016) (Variables|Default|GetScene%20-%2017) (Variables|Default|GetScene%20-%2018) (Variables|Default|GetScene%20-%2019) (Variables|Default|GetScene%20-%2020))) ; list of teleport locations for players
  (Variables|Default|SetEditableLocationsArray (Variables|Default|GetTeleport%20Locations)))
