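; Server-authoritative event zone for the GrillMoon live event, switched on or off by calendar, day-event router messages or cheats.
; Activation toggles overlap collision, broadcasts zone state and controls a respawn mutator using the zone's respawn area.
; Overlap applies aura, immunity and gravity changes, disables vehicles and clears sprites and pre-event pickups, with exit reversing player effects.
; Day 11 events drive an end-phase blocking mutator, global HUD hiding with DBNO revives, and post-event mode.
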(event Custom|RemoveOutsideHUDEffect
  (bind _outfortplayerpawns (Fortnite|GetFortPlayerPawns))
  (for _array_element _outfortplayerpawns
    (if (Pawn|IsBotControlled _array_element)
      (GameplayEffects|RemoveActiveGameplayEffectbySourceEffect (Ability|GetFortAbilitySystemComponent _array_element) "/GrillMoon/Gameplay/Effects/GE_GrillMoon_EventAuraBots.GE_GrillMoon_EventAuraBots_C" 0 -1)
      (else
        (GameplayEffects|RemoveActiveGameplayEffectbySourceEffect (Ability|GetFortAbilitySystemComponent _array_element) "/GrillMoon/Gameplay/Effects/GE_GrillMoon_EventAura.GE_GrillMoon_EventAura_C" 0 -1)
        (GameplayEffects|RemoveActiveGameplayEffectbySourceEffect (Ability|GetFortAbilitySystemComponent _array_element) "/GrillMoon/Gameplay/Effects/GE_GrillMoon_EventAuraFX.GE_GrillMoon_EventAuraFX_C" 0 -1)))))

(event EventBeginPlay
  (bind _returnvalue (GameplayEventRouter|GetGlobalEventRouter))
  ;[0 0.15 0.1039 0.5] Activate Zone via Calendar Event
  (Networking|SwitchHasAuthority
    (:Authority
      (Class|FortAsyncActionAuthorityMatchReady|MatchReadyAsync
        (:OnReady
          (MCPContext|InitialCalendarSyncCompleted
            (:OnSyncComplete
              (MCPContext|WatchCalendarEvent (Variables|Default|GetEventZoneCalendarEventName)
                (:then
                  ;[0 0.15 0.03011 0.5] If event is inactive, allow zone to be activated via cheats.
                  (bind (_asynctaskproxy _payload) (GameplayEventRouter|StartListeningToStatefulEvent _returnvalue "/GrillMoon_Countdown/Gameplay/EventRouter/GrillMoon_Day_Event.GrillMoon_Day_Event")
                    (:OnEventReceived
                      _)
                    (:OnSavedEventStateExists
                      _))
                  (switch Utilities|FlowControl|Switch|SwitchonGrillMoon_Day_Event_Enum (Utilities|Struct|BreakGrillMoonDayEvent _payload)
                    (:NewEnumerator12
                      (CallFunction|ZoneEvent true))
                    (:NewEnumerator13
                      (CallFunction|ZoneEvent true)
                      (bind (_asynctaskproxy_2 _payload_2) (GameplayEventRouter|StartListeningToStatefulEvent _returnvalue "/GrillMoon/Gameplay/EventRouter/GrillMoon_Day11_Events.GrillMoon_Day11_Events")
                        (:then
                          (bind _returnvalue_2 (Class|SpecialEventGameplayLibrary|SpawnMutator "/GrillMoon/Gameplay/M_SE_GM_BlockEndPhase.M_SE_GM_BlockEndPhase_C"))
                          (bind _output_get (Variables|Default|SetSE_BlockEndPhase _returnvalue_2))
                          (Mutator|SetMutatorActive _output_get true))
                        (:OnEventReceived
                          (switch Utilities|FlowControl|Switch|SwitchonGrillMoon_Day11_Events_Enum (Utilities|Struct|BreakGrillMoonDay11Events _payload_2)
                            (:NewEnumerator9
                              (CallFunction|HideHUDForAllOutside))
                            (:NewEnumerator1
                              (CallFunction|RemoveOutsideHUDEffect))
                            (:NewEnumerator6
                              (Variables|Default|SetPostEventMode true))
                            (:NewEnumerator3
                              (bind _returnvalue_3 (Mutator|GetGameplayMutatorbyClass "/GrillMoon/Gameplay/M_SE_GM_BlockEndPhase.M_SE_GM_BlockEndPhase_C"))
                              (Mutator|SetMutatorActive _returnvalue_3 false)
                              (bind _returnvalue_4 (Mutator|GetGameplayMutatorbyClass "/GrillMoon/Gameplay/M_SE_GrillMoon_Respawn.M_SE_GrillMoon_Respawn_C"))
                              (Mutator|SetMutatorActive _returnvalue_4 false))))))))
                (:OnEventActive
                  (CallFunction|ZoneEvent true))
                (:OnEventEnded
                  (CallFunction|ZoneEvent false))))))))))

(event Custom|GM_EventZoneOn
  ;[0 0.15 0.05764 0.5] Activate/Deactivate Zone via cheats
  (Networking|SwitchHasAuthority
    (:Authority
      (CallFunction|ZoneEvent true))))

(event Custom|GM_EventZoneOff
  ;[0 0.15 0.05764 0.5] Activate/Deactivate Zone via cheats
  (Networking|SwitchHasAuthority
    (:Authority
      (CallFunction|ZoneEvent false))))

(event Custom|AddCombatModeEffectToPlayer
  (params (ActorEnteringCombatMode :type Actor%20Object%20Reference)))

(event OnComponentBeginOverlap(EventZone)
  (params (OverlappedComponent :type Primitive%20Component%20Object%20Reference) (OtherActor :type Actor%20Object%20Reference) (OtherComp :type Primitive%20Component%20Object%20Reference) (OtherBodyIndex :type Integer) (bFromSweep :type Boolean) (SweepResult :type Hit%20Result%20Structure%20%28by%20ref%29))
  ;[0 0.15 0.09335 0.5] Apply Zone effects when players begin overlap
  (Networking|SwitchHasAuthority
    (:Authority
      (if (and (Utilities|ClassIsChildOf (Utilities|GetClass OtherActor) "/Script/FortniteGame.FortPickupAthena") (not (Variables|Default|GetPostEventMode)))
        (Actor|DestroyActor OtherActor)
        (else
          (CallFunction|ApplyZoneEffectstoPlayer OtherActor)
          (Utilities|IsValid OtherActor
            (:Is%20Valid
              (CallFunction|DisableVehicle true OtherActor)
              (CallFunction|SpriteCheck OtherActor))))))))

(event OnComponentEndOverlap(EventZone)
  (params (OverlappedComponent :type Primitive%20Component%20Object%20Reference) (OtherActor :type Actor%20Object%20Reference) (OtherComp :type Primitive%20Component%20Object%20Reference) (OtherBodyIndex :type Integer))
  ;[0 0.15 0.09335 0.5] Apply Zone effects when players begin overlap
  (Networking|SwitchHasAuthority
    (:Authority
      (CallFunction|RemoveZoneEffectsfromPlayer OtherActor)
      (CallFunction|DisableVehicle false OtherActor))))

(event Custom|ZoneEvent
  (params (Activate :type Boolean))
  (bind _respawnmutator (Variables|Default|GetRespawnMutator))
  ;[0 0.15 0.09362 0.5] Activate Collision for Event Zone and set Zone Enabled to True.
  (Collision|SetCollisionEnabled (Variables|Default|GetEventZone) (select Activate "QueryOnly" "NoCollision"))
  (GameplayEventRouter|BroadcastStatefulEvent (GameplayEventRouter|GetGlobalEventRouter) (Utilities|Struct|MakeGrillMoonEventZoneActive Activate) 0)
  (if (Variables|Default|GetPostEventMode)
    _
    (elif Activate
      (Utilities|IsValid _respawnmutator
        (:Is%20Valid
          _)
        (:Is%20Not%20Valid
          (bind _returnvalue (Class|SpecialEventGameplayLibrary|SpawnMutator "/GrillMoon/Gameplay/M_SE_GrillMoon_Respawn.M_SE_GrillMoon_Respawn_C"))
          (Variables|Default|SetRespawnMutator _returnvalue)))
      (Mutator|SetMutatorActive _respawnmutator true)
      (bind _asm_se_grill_moon_respawn (Utilities|Casting|CastToM_SE_GrillMoon_Respawn _respawnmutator))
      (Class|FortAthenaMutatorSERespawn|OverrideRespawnArea _asm_se_grill_moon_respawn (Variables|Default|GetRespawn%20Area))
      (else
        (bind _returnvalue_2 (Mutator|GetGameplayMutatorbyClass "/GrillMoon/Gameplay/M_SE_GrillMoon_Respawn.M_SE_GrillMoon_Respawn_C"))
        (Mutator|SetMutatorActive _returnvalue_2 false)))))

(event Custom|DisableVehicle
  (params (Disable :type Boolean) (Actor :type Actor%20Object%20Reference))
  (bind _gameplay_tags (Variables|Default|GetGameplay%20Tags))
  (bind _asfort_athena_vehicle (Utilities|Casting|CastToFortAthenaVehicle Actor))
  (if Disable
    (Vehicle|WakeUp _asfort_athena_vehicle)
    (Components|Movement|StopMovementImmediately (Pawn|GetMovementComponent _asfort_athena_vehicle))
    (Ability|GameplayEffect|AddLooseGameplayTags Actor _gameplay_tags true)
    (else
      (Ability|GameplayEffect|RemoveLooseGameplayTags Actor _gameplay_tags true))))

(event Custom|SpriteCheck
  (params (Actor :type Actor%20Object%20Reference))
  (bind _returnvalue (Utilities|GetClass Actor))
  ; Check if overlap is a Sprite
  (if (Utilities|ClassIsChildOf _returnvalue "/Sprites/Gameplay/ItemVersion/Abilities/Prj_Athena_Sprite_Parent.Prj_Athena_Sprite_Parent_C")
    (Actor|DestroyActor Actor)
    (else
      (if (Utilities|ClassIsChildOf _returnvalue "/Sprites/NPCs/AISpawnerData/Pawn/NPC_Pawn_Irwin_Simple_Sprite.NPC_Pawn_Irwin_Simple_Sprite_C")
        (Game|Damage|ApplyDamage Actor 5e+02)))))

(event Custom|HideHUDForAllOutside
  ; Hide HUD for Outside + Rvieve from DBNO
  (Networking|SwitchHasAuthority
    (:Authority
      (CallFunction|ListenForForceReviveEvent)
      (bind _outfortplayerpawns (Fortnite|GetFortPlayerPawns))
      (Development|PrintString (Utilities|String|Append "I found " (Utilities|String|ToString(Integer) (Utilities|Array|Length _outfortplayerpawns)) " players.") true true "(R=0.000000,G=0.660000,B=1.000000,A=1.000000)" 5e+01)
      (for _array_element _outfortplayerpawns
        (if (Health|IsDBNO _array_element)
          (bind _asfort_player_pawn_athena (Utilities|Casting|CastToFortPlayerPawnAthena _array_element)
            (:then
              (Class|FortPlayerPawnAthena|ForceRevivefromDBNO _asfort_player_pawn_athena 0) ; this happens about 8 seconds before the teleport\nwhen we hide HUDs
              (FortPawn|SetHealth _array_element (FortPawn|GetMaxHealth _array_element))
              (FortPawn|SetShield _array_element (FortPawn|GetMaxShield _array_element)))
            (:CastFailed
              _))
          (if (Abilities|HasAbilitywithAnyMatchingTag (Ability|GetFortAbilitySystemComponent _array_element) (Variables|Default|GetHide%20Hud%20Tag) false)
            (Development|PrintString "HUD is already hidden for this player" true true "(R=0.000000,G=0.660000,B=1.000000,A=1.000000)" 3e+01)
            (elif (not (Pawn|IsBotControlled _array_element))
              (Development|PrintString "Player can still see hud" true true "(R=0.000000,G=0.660000,B=1.000000,A=1.000000)" 3e+01)
              (GameplayEffects|ApplyGameplayEffectToTarget :self (Ability|GetFortAbilitySystemComponent _array_element) :GameplayEffectClass "/GrillMoon/Gameplay/Effects/GE_GrillMoon_EventAura.GE_GrillMoon_EventAura_C" :Target (Ability|GetFortAbilitySystemComponent _array_element) :Level 0.0)
              (GameplayEffects|ApplyGameplayEffectToTarget :self (Ability|GetFortAbilitySystemComponent _array_element) :GameplayEffectClass "/GrillMoon/Gameplay/Effects/GE_GrillMoon_EventAuraFX.GE_GrillMoon_EventAuraFX_C" :Target (Ability|GetFortAbilitySystemComponent _array_element) :Level 0.0)
              (else
                (GameplayEffects|ApplyGameplayEffectToTarget :self (Ability|GetFortAbilitySystemComponent _array_element) :GameplayEffectClass "/GrillMoon/Gameplay/Effects/GE_GrillMoon_EventAuraBots.GE_GrillMoon_EventAuraBots_C" :Target (Ability|GetFortAbilitySystemComponent _array_element) :Level 0.0)))))))))

(event Custom|GM_TestRevive
  ; Hide HUD for Outside + Rvieve from DBNO
  ;; test console commands
  (Networking|SwitchHasAuthority
    (:Authority
      (bind _outfortplayerpawns (Fortnite|GetFortPlayerPawns))
      (for _array_element _outfortplayerpawns
        (if (Health|IsDBNO _array_element)
          (FortPlayerPawn|RevivefromDBNO _array_element 0) ; doesn't do anything
        )))))

(event Custom|Listen%20for%20Force%20Revive%20Event
  ; Hide HUD for Outside + Rvieve from DBNO
  (Networking|SwitchHasAuthority
    (:Authority
      (GameplayEventRouter|StartListeningToStatefulEvent (GameplayEventRouter|GetGlobalEventRouter) "/GrillMoon/Gameplay/EventRouter/GrillMoon_ForceRevive_Event.GrillMoon_ForceRevive_Event"
        (:OnEventReceived
          _)
        (:OnSavedEventStateExists
          _))
      (CallFunction|ForceReviveAllPlayers))))

(event Custom|GM_TestManualRevive
  ; Hide HUD for Outside + Rvieve from DBNO
  ;; test console commands
  (Networking|SwitchHasAuthority
    (:Authority
      (bind _outfortplayerpawns (Fortnite|GetFortPlayerPawns))
      (for _array_element _outfortplayerpawns
        (if (Health|IsDBNO _array_element)
          (FortPlayerPawn|ManualRevivefromDBNO _array_element 0) ; revives them, but the UI is stuck in the DBNO state
        )))))

(event Custom|GM_ReviveHealth
  ; Hide HUD for Outside + Rvieve from DBNO
  ;; test console commands
  (Networking|SwitchHasAuthority
    (:Authority
      (bind _outfortplayerpawns (Fortnite|GetFortPlayerPawns))
      (for _array_element _outfortplayerpawns
        (FortPawn|SetHealth _array_element (FortPawn|GetMaxHealth _array_element))
        (FortPawn|SetShield _array_element (FortPawn|GetMaxShield _array_element))))))

(event Custom|ForceReviveAllPlayers
  ; Hide HUD for Outside + Rvieve from DBNO
  (bind _outfortplayerpawns (Fortnite|GetFortPlayerPawns))
  (for _array_element _outfortplayerpawns
    (if (Health|IsDBNO _array_element)
      (bind _asfort_player_pawn_athena (Utilities|Casting|CastToFortPlayerPawnAthena _array_element))
      (Class|FortPlayerPawnAthena|ForceRevivefromDBNO _asfort_player_pawn_athena 0) ; this happens about 8 seconds before the teleport\nwhen we hide HUDs
      (FortPawn|SetHealth _array_element (FortPawn|GetMaxHealth _array_element))
      (FortPawn|SetShield _array_element (FortPawn|GetMaxShield _array_element)))))

(event Custom|GM_TestManualRevivePlus%20Health
  ; Hide HUD for Outside + Rvieve from DBNO
  ;; test console commands
  (Networking|SwitchHasAuthority
    (:Authority
      (bind _outfortplayerpawns (Fortnite|GetFortPlayerPawns))
      (for _array_element _outfortplayerpawns
        (if (Health|IsDBNO _array_element)
          (FortPlayerPawn|ManualRevivefromDBNO _array_element 0) ; revives, but the HUD is stuck in the DBNO state / colors
          (FortPawn|SetHealth _array_element (FortPawn|GetMaxHealth _array_element))
          (FortPawn|SetShield _array_element (FortPawn|GetMaxShield _array_element)))))))

(event Custom|GM_TestForceRevive
  ; Hide HUD for Outside + Rvieve from DBNO
  ;; test console commands
  (Networking|SwitchHasAuthority
    (:Authority
      (bind _outfortplayerpawns (Fortnite|GetFortPlayerPawns))
      (for _array_element _outfortplayerpawns
        (if (Health|IsDBNO _array_element)
          (bind _asfort_player_pawn_athena (Utilities|Casting|CastToFortPlayerPawnAthena _array_element))
          (Class|FortPlayerPawnAthena|ForceRevivefromDBNO _asfort_player_pawn_athena 0) ; this works
        )))))

(fn ApplyZoneEffectsToPlayer
  (params (Actor :type Actor%20Object%20Reference))
  (bind _returnvalue (Ability|GetFortAbilitySystemComponent Actor))
  ;[0 0.15 0.1302 0.5] Modify Gravity scale, and apply Zone GEs
  (bind _asfort_player_pawn_athena (Utilities|Casting|CastToFortPlayerPawnAthena Actor))
  (if (not (Pawn|IsBotControlled _asfort_player_pawn_athena))
    (FortPawn|SetGravityMultiplier _asfort_player_pawn_athena (ScalableFloat|GetValueatLevelwithContext (Variables|Default|GetGravityScaleMultiplier) 1.0))
    (Utilities|IsValid _returnvalue
      (:Is%20Valid
        (GameplayEffects|ApplyGameplayEffectToTarget :self _returnvalue :GameplayEffectClass (select (Variables|Default|GetPostEventMode) (Variables|Default|GetPostEventAura) (Variables|Default|GetPreEventAura)) :Target _returnvalue :Level 0.0)
        (GameplayEffects|ApplyGameplayEffectToTarget :self _returnvalue :GameplayEffectClass "/GrillMoon/Gameplay/Effects/GE_GrillMoon_EventAuraFX.GE_GrillMoon_EventAuraFX_C" :Target _returnvalue :Level 0.0)
        (GameplayEffects|ApplyGameplayEffectToTarget :self _returnvalue :GameplayEffectClass "/GrillMoon/Gameplay/Effects/GE_GrillMoon_EventDamageImmune.GE_GrillMoon_EventDamageImmune_C" :Target _returnvalue :Level 0.0))
      (:Is%20Not%20Valid
        (Fortnite|FortShippingLog self "GrillMoonEventZone: Overlapping Actor does not have a Fort Ability System Component.")))
    (else
      (GameplayEffects|ApplyGameplayEffectToTarget :self (Ability|GetFortAbilitySystemComponent Actor) :GameplayEffectClass "/GrillMoon/Gameplay/Effects/GE_GrillMoon_EventAuraBots.GE_GrillMoon_EventAuraBots_C" :Target (Ability|GetFortAbilitySystemComponent Actor) :Level 0.0))))

(fn RemoveZoneEffectsFromPlayer
  (params (ActorEndingOverlap :type Actor%20Object%20Reference))
  (bind _returnvalue (Ability|GetFortAbilitySystemComponent ActorEndingOverlap))
  ;[0 0.15 0.1277 0.5] Remove Zone related Gameplay effects, and set Gravity scale to default.
  (bind _asfort_player_pawn_athena (Utilities|Casting|CastToFortPlayerPawnAthena ActorEndingOverlap))
  (FortPawn|SetGravityMultiplier _asfort_player_pawn_athena (Variables|Default|GetDefaultGravityScale))
  (Utilities|IsValid _returnvalue
    (:Is%20Valid
      (GameplayEffects|RemoveActiveGameplayEffectbySourceEffect _returnvalue "/GrillMoon/Gameplay/Effects/GE_GrillMoon_EventAuraFX.GE_GrillMoon_EventAuraFX_C" 0 -1)
      (GameplayEffects|RemoveActiveGameplayEffectbySourceEffect _returnvalue (Variables|Default|GetPostEventAura) 0 -1)
      (GameplayEffects|RemoveActiveGameplayEffectbySourceEffect _returnvalue (Variables|Default|GetPreEventAura) 0 -1)
      (GameplayEffects|RemoveActiveGameplayEffectbySourceEffect _returnvalue "/GrillMoon/Gameplay/Effects/GE_GrillMoon_EventDamageImmune.GE_GrillMoon_EventDamageImmune_C" 0 -1)
      (GameplayEffects|RemoveActiveGameplayEffectbySourceEffect _returnvalue "/GrillMoon/Gameplay/Effects/GE_GrillMoon_CombatAura.GE_GrillMoon_CombatAura_C" 0 -1)
      (GameplayEffects|RemoveActiveGameplayEffectbySourceEffect _returnvalue "/GrillMoon/Gameplay/Effects/GE_GrillMoon_EventAuraBots.GE_GrillMoon_EventAuraBots_C" 0 -1)
      (if (Utilities|Enum|Equal(Enum) (Variables|/Script/Engine.CharacterMovementComponent|GetMovementMode (Variables|/Script/Engine.Character|GetCharacterMovement _asfort_player_pawn_athena)) "MOVE_Falling")
        (GameplayEffects|ApplyGameplayEffectToTarget :self _returnvalue :GameplayEffectClass "/Game/Athena/Items/Parents/FallImmunity/GE_Athena_TillLandFallDamageImmunity_Grant_ApplyInAirOnly.GE_Athena_TillLandFallDamageImmunity_Grant_ApplyInAirOnly_C" :Target _returnvalue :Level 0.0)))))

(fn ConstructionScript)
