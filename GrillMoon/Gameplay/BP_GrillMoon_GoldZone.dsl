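; GrillMoon Gold Zone: an event area that gives overlapping actors with a Fort ability system the GoldAura gameplay effect.
; On the server, a calendar event, watched after match ready and calendar sync, or cheat events toggle the zone's query collision.
; End overlap removes the effect and resets pawn gravity, though entry never changes gravity here.
; Gold junk rift spawner logic is mostly empty stubs or marked disabled, with leftover OnRep and range-check helpers.
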(event Custom|GM_GoldZoneOn
  ;[0 0.15 0.05764 0.5] Activate/Deactivate Zone via cheats
  (Networking|SwitchHasAuthority
    (:Authority
      (CallFunction|ZoneEvent true))))

(event EventBeginPlay
  ;[0 0.15 0.1039 0.5] Activate Zone via Calendar Event
  (Networking|SwitchHasAuthority
    (:Authority
      (Class|FortAsyncActionAuthorityMatchReady|MatchReadyAsync
        (:OnReady
          (MCPContext|InitialCalendarSyncCompleted
            (:OnSyncComplete
              (MCPContext|WatchCalendarEvent (Variables|Default|GetEventZoneCalendarEventName)
                (:then
                  (CallFunction|ZoneEvent true))
                (:OnEventEnded
                  (CallFunction|ZoneEvent false))))))))))

(event Custom|GM_GoldZoneOff
  ;[0 0.15 0.05764 0.5] Activate/Deactivate Zone via cheats
  (Networking|SwitchHasAuthority
    (:Authority
      (CallFunction|ZoneEvent false))))

(event Custom|ZoneEvent
  (params (Activate :type Boolean))
  ;[0 0.15 0.09362 0.5] Activate Collision for Event Zone
  (Collision|SetCollisionEnabled (Variables|Default|GetEventZone) (select Activate "QueryOnly" "NoCollision")))

(event OnComponentBeginOverlap(EventZone)
  (params (OverlappedComponent :type Primitive%20Component%20Object%20Reference) (OtherActor :type Actor%20Object%20Reference) (OtherComp :type Primitive%20Component%20Object%20Reference) (OtherBodyIndex :type Integer) (bFromSweep :type Boolean) (SweepResult :type Hit%20Result%20Structure%20%28by%20ref%29))
  ;[0 0.15 0.09335 0.5] Apply Zone effects when players begin overlap
  (Networking|SwitchHasAuthority
    (:Authority
      (CallFunction|ApplyZoneEffectstoPlayer OtherActor))))

(event Custom|GM_ZoneGoldJunkRiftsOn)

(event EventEndPlay
  (params (EndPlayReason :type EEndPlayReason%20Enum :default "Destroyed")))

(event OnComponentEndOverlap(EventZone)
  (params (OverlappedComponent :type Primitive%20Component%20Object%20Reference) (OtherActor :type Actor%20Object%20Reference) (OtherComp :type Primitive%20Component%20Object%20Reference) (OtherBodyIndex :type Integer))
  ;[0 0.15 0.09335 0.5] Apply Zone effects when players begin overlap
  (Networking|SwitchHasAuthority
    (:Authority
      (CallFunction|RemoveZoneEffectsfromPlayer OtherActor))))

(event Custom|GM_ZoneGoldJunkRiftsOff)

(event Custom|ActivateJunkSpawners)

(event Custom|SpawnRandomRiftOnFirstOverlap
  ;[0 0 0 0.5] Activate Junk Spawners - Disabled
  (Variables|Default|SetChosenSpawner (Utilities|Array|RandomArrayItem (Variables|Default|GetGoldJunkSpawners))))

(event Custom|GrillMoon_SpawnNearestJunkRift)

(event Custom|CheckIfPlayersInRange)

(event Custom|GrillMoon_ShowGoldJunkSpawners)

(event Custom|GrillMoon_HideGoldJunkSpawners)

(event Custom|GrillMoon_ActivateRandomJunkRift
  ;[0 0 0 0.5] See if any players are in range of the Gold Zone - Disabled
  (CallFunction|ActivateJunkSpawners))

(fn ApplyZoneEffectsToPlayer
  (params (Actor :type Actor%20Object%20Reference))
  (bind _returnvalue (Ability|GetFortAbilitySystemComponent Actor))
  ;[0 0.15 0.1302 0.5] Modify Gravity scale, and apply Zone GEs
  (Utilities|IsValid _returnvalue
    (:Is%20Valid
      (GameplayEffects|ApplyGameplayEffectToTarget :self _returnvalue :GameplayEffectClass "/GrillMoon/Gameplay/Effects/GE_GrillMoon_GoldAura.GE_GrillMoon_GoldAura_C" :Target _returnvalue :Level 0.0))
    (:Is%20Not%20Valid
      (Fortnite|FortShippingLog self "GrillMoonEventZone: Overlapping Actor does not have a Fort Ability System Component."))))

(fn IsActorAnItemOrChest
  (params (ActorToCheck :type Actor%20Object%20Reference))
  (bind _actortocheck (Variables|GetActorToCheck))
  (bind _returnvalue (Utilities|GetClass _actortocheck))
  (Variables|Default|SetActorMatchesClass true)
  (for _array_element (Variables|Default|GetItemAndChestClasses)
    (if (or (Utilities|ClassIsChildOf _returnvalue _array_element) (== _returnvalue _array_element))
      (if (!= _returnvalue "/Game/Athena/Items/Gameplay/Wads/B_Pickups_Wads.B_Pickups_Wads_C")
        (Variables|Default|SetActorMatchesClass true))))
  (if (Variables|Default|GetActorMatchesClass)
    (return true)
    (else
      (return false))))

(fn IsAValidPlayerWithinRange
  (params (PlayerState :type Player%20State%20Object%20Reference))
  (bind _returnvalue (PlayerState|GetPawn PlayerState))
  (Utilities|IsValid PlayerState
    (:Is%20Valid
      (Utilities|IsValid _returnvalue
        (:Is%20Valid
          (if (Pawn|IsPlayerControlled _returnvalue)
            (if (<= (Transformation|GetDistanceTo _returnvalue self) (ScalableFloat|GetValueatLevelwithContext (Variables|Default|GetMaxCheckDistance)))
              (return true)
              (else
                _))
            (else
              _))
          (return false))
        (:Is%20Not%20Valid
          _)))
    (:Is%20Not%20Valid
      _)))

(fn OnRep_ChosenSpawner
  (bind _chosenspawner (Variables|Default|GetChosenSpawner))
  (Utilities|IsValid _chosenspawner
    (:Is%20Valid
      (Class|BAthenaJollyRascalSpawnerGM|ActivatePortal _chosenspawner))))

(fn OnRep_ShowSpawners
  (Networking|CanExecuteCosmeticEvents
    (:True
      (for _array_element (Variables|Default|GetGoldJunkSpawners)
        (Rendering|SetVisibility (Variables|/GrillMoon/Gameplay/B_Athena_JollyRascal_Spawner_GM.B_Athena_JollyRascal_Spawner_GM_C|GetLocationIndicator _array_element) (Variables|Default|GetShowSpawners) false)))))

(fn RemoveZoneEffectsFromPlayer
  (params (ActorEndingOverlap :type Actor%20Object%20Reference))
  (bind _returnvalue (Ability|GetFortAbilitySystemComponent ActorEndingOverlap))
  ;[0 0.15 0.1277 0.5] Remove Zone related Gameplay effects, and set Gravity scale to default.
  (bind _asfort_player_pawn_athena (Utilities|Casting|CastToFortPlayerPawnAthena ActorEndingOverlap))
  (FortPawn|SetGravityMultiplier _asfort_player_pawn_athena (Variables|Default|GetDefaultGravityScale))
  (Utilities|IsValid _returnvalue
    (:Is%20Valid
      (GameplayEffects|RemoveActiveGameplayEffectbySourceEffect _returnvalue "/GrillMoon/Gameplay/Effects/GE_GrillMoon_GoldAura.GE_GrillMoon_GoldAura_C" 0 -1))))

(fn ConstructionScript)
