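; GrillMoon knockback actor that, when switched on by a stateful gameplay event or calendar event, pushes players and vehicles away from its location.
; Radius and capsule overlaps filter out AI and seated players, then apply server-side outward-and-up impulses.
; A looping timer rechecks capsule overlaps to catch boosting vehicles, which also receive a boost-cancelling force.
; While active, it optionally destroys nearby player-built structures every two seconds.
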(event Custom|ListenForEvents
  ;[0 1 0.5985 0.25] Listen out for activation/deactivation events.
  (switch Utilities|FlowControl|Switch|SwitchonGM_Knockback_EventTypes (Variables|Default|GetEventType)
    (:NewEnumerator0
      _)
    (:NewEnumerator2
      (Utilities|FlowControl|Sequence
        (:then_0
          _)
        (:then_1
          (bind (_asynctaskproxy _payload) (GameplayEventRouter|StartListeningToStatefulEvent (GameplayEventRouter|GetGlobalEventRouter) "/GrillMoon/Gameplay/GameplayEvents/GrillMoon_Knockback_Activation.GrillMoon_Knockback_Activation")
            (:OnSavedEventStateExists
              (if (Utilities|Struct|BreakGrillMoonKnockbackActivation _payload)
                (CallFunction|ActivateKnockback)
                (else
                  (CallFunction|DisableKnockback))))))))
    (:NewEnumerator1
      _))
  (MCPContext|WatchCalendarEvent (Variables|Default|GetCalendarEventToWatch)
    (:OnEventActive
      (CallFunction|ActivateKnockback))
    (:OnEventEnded
      (CallFunction|DisableKnockback))))

(event Custom|GrillMoon_ActivateKnockback
  ;[0 1 0.5985 0.25] Activate when events call
  (Collision|SetCollisionEnabled (Variables|Default|GetKnockbackRadius) "QueryOnly")
  (if (Variables|Default|GetDestroyBuildingsInKnockback)
    (bind _returnvalue (Utilities|Time|SetTimerbyEvent (EventDispatchers|CreateEvent :!event Custom|DestroyPlayerBuilding) 2.0 true))
    (Variables|Default|SetBuildingTimer _returnvalue)))

(event Custom|PlayKnockbackSound
  ;[0 1 0.5985 0.25] Play Knockback Sound - Not Used in Deckbed
  (Audio|PlaySoundatLocation (Variables|Default|GetKnockback%20Sound) (Transformation|GetActorLocation (Variables|Default|GetOverlappingActor))))

(event Custom|GrillMoon_DisableKnockback
  ;[0 1 0.6312 0.25] Every 2 seconds, check for player placed buildings in a radius, and destroy them.
  (Utilities|Time|ClearandInvalidateTimerbyHandle (Variables|Default|GetBuildingTimer))
  (Utilities|Time|ClearandInvalidateTimerbyHandle (Variables|Default|GetPostKnockbackOverlapCheckTimer))
  (Collision|SetCollisionEnabled (Variables|Default|GetKnockbackRadius) "NoCollision"))

(event Custom|ActivateKnockback
  ;[0 1 0.5985 0.25] Activate when events call
  (Collision|SetCollisionEnabled (Variables|Default|GetKnockbackRadius) "QueryOnly")
  (if (Variables|Default|GetDestroyBuildingsInKnockback)
    (bind _returnvalue (Utilities|Time|SetTimerbyEvent (EventDispatchers|CreateEvent :!event Custom|DestroyPlayerBuilding) 2.0 true))
    (Variables|Default|SetBuildingTimer _returnvalue)))

(event EventBeginPlay
  ;[0 1 0.6312 0.25] Every 2 seconds, check for player placed buildings in a radius, and destroy them.
  (|CallParentFunction)
  (Collision|SetCollisionEnabled (Variables|Default|GetKnockbackRadius) "NoCollision")
  (CallFunction|ListenforEvents))

(event Custom|DisableKnockback
  ;[0 1 0.6312 0.25] Every 2 seconds, check for player placed buildings in a radius, and destroy them.
  (Utilities|Time|ClearandInvalidateTimerbyHandle (Variables|Default|GetBuildingTimer))
  (Utilities|Time|ClearandInvalidateTimerbyHandle (Variables|Default|GetPostKnockbackOverlapCheckTimer))
  (Collision|SetCollisionEnabled (Variables|Default|GetKnockbackRadius) "NoCollision"))

(event EventEndPlay
  (params (EndPlayReason :type EEndPlayReason%20Enum :default "Destroyed"))
  ;[0 1 0.6312 0.25] Every 2 seconds, check for player placed buildings in a radius, and destroy them.
  (Utilities|Time|ClearandInvalidateTimerbyHandle (Variables|Default|GetBuildingTimer))
  (Utilities|Time|ClearandInvalidateTimerbyHandle (Variables|Default|GetPostKnockbackOverlapCheckTimer))
  (Collision|SetCollisionEnabled (Variables|Default|GetKnockbackRadius) "NoCollision"))

(event Custom|DestroyPlayerBuilding
  ;[0 1 0.6312 0.25] Every 2 seconds, check for player placed buildings in a radius, and destroy them.
  (Networking|SwitchHasAuthority
    (:Authority
      (bind _outactors (Collision|SphereOverlapActors (Transformation|GetActorLocation self) (Variables|Default|GetDestroyRadius) (Variables|Default|GetBuildings) "/Script/FortniteGame.BuildingActor"))
      (for _array_element _outactors
        (bind _asbuilding_actor (Utilities|Casting|CastToBuildingActor _array_element))
        (if (Variables|/Script/FortniteGame.BuildingActor|GetbPlayerPlaced _asbuilding_actor)
          (Actor|DestroyActor _asbuilding_actor))))))

(event Custom|%20CheckForOverlappingActors
  (bind _knockbackcapsule (Variables|Default|GetKnockbackCapsule))
  (bind _overlappingactors (Collision|GetOverlappingActors _knockbackcapsule 0 :!class /Script/Engine.PrimitiveComponent))
  ;[0 1 0.5144 0.25] Check for overlapping Vehicles that managed to boost through intial knockback attempt
  (if (> (Utilities|Array|Length _overlappingactors) 0)
    (for _array_element _overlappingactors
      (CallFunction|FilterActorforKnockback _array_element))
    (else
      (Utilities|Time|ClearandInvalidateTimerbyHandle (Variables|Default|GetPostKnockbackOverlapCheckTimer)))))

(event Custom|FilterActorForKnockback
  (params (ActorToFilter :type Actor%20Object%20Reference))
  ;[0 1 0.6647 0.25] Filter overlapping actor
  (Utilities|IsValid ActorToFilter
    (:Is%20Valid
      (Variables|Default|SetOverlappingActor ActorToFilter)
      (bind _asfort_player_pawn (Utilities|Casting|CastToFortPlayerPawn ActorToFilter)
        (:then
          (if (not (Vehicle|IsinVehicle _asfort_player_pawn))
            (CallFunction|KnockBackPawn _asfort_player_pawn)))
        (:CastFailed
          (Utilities|Casting|CastToFortAIPawn ActorToFilter
            (:CastFailed
              (bind _asfort_athena_vehicle (Utilities|Casting|CastToFortAthenaVehicle ActorToFilter))
              (CallFunction|KnockBackVehicle _asfort_athena_vehicle ActorToFilter))))))))

(event OnComponentBeginOverlap(KnockbackRadius)
  (params (OverlappedComponent :type Primitive%20Component%20Object%20Reference) (OtherActor :type Actor%20Object%20Reference) (OtherComp :type Primitive%20Component%20Object%20Reference) (OtherBodyIndex :type Integer) (bFromSweep :type Boolean) (SweepResult :type Hit%20Result%20Structure%20%28by%20ref%29))
  ;[0 1 0.7834 0.25] Get overlapping actor and attempt to apply knockback.
  (CallFunction|FilterActorforKnockback OtherActor)
  (if (not (Utilities|Time|IsValidTimerHandle (Variables|Default|GetPostKnockbackOverlapCheckTimer)))
    (bind _returnvalue (Utilities|Time|SetTimerbyEvent (EventDispatchers|CreateEvent :!event Custom|%20CheckForOverlappingActors) (Variables|Default|GetKnockbackLoopFrequency) true))
    (Variables|Default|SetPostKnockbackOverlapCheckTimer _returnvalue)))

(event OnComponentBeginOverlap(KnockbackCapsule)
  (params (OverlappedComponent :type Primitive%20Component%20Object%20Reference) (OtherActor :type Actor%20Object%20Reference) (OtherComp :type Primitive%20Component%20Object%20Reference) (OtherBodyIndex :type Integer) (bFromSweep :type Boolean) (SweepResult :type Hit%20Result%20Structure%20%28by%20ref%29))
  ;[0 1 0.7834 0.25] Get overlapping actor and attempt to apply knockback.
  (CallFunction|FilterActorforKnockback OtherActor)
  (if (not (Utilities|Time|IsValidTimerHandle (Variables|Default|GetPostKnockbackOverlapCheckTimer)))
    (bind _returnvalue (Utilities|Time|SetTimerbyEvent (EventDispatchers|CreateEvent :!event Custom|%20CheckForOverlappingActors) (Variables|Default|GetKnockbackLoopFrequency) true))
    (Variables|Default|SetPostKnockbackOverlapCheckTimer _returnvalue)))

(event Custom|KnockBackPawn
  (params (Pawn :type Fort%20Pawn%20Object%20Reference))
  (bind _charactermovement (Variables|/Script/Engine.Character|GetCharacterMovement Pawn))
  ;[0 1 0.5479 0.25] Knockback Pawns on server side.
  (Networking|SwitchHasAuthority
    (:Authority
      (if (GameplayTags|HasMatchingGameplayTag Pawn "(TagName=\"Ability.Moonflax.NitroGauntlets.Meteor\")")
        (Ability|GameplayEffect|RemoveLooseGameplayTags Pawn (Variables|Default|GetMovementAbilityTag) true))
      (Components|Movement|StopMovementImmediately _charactermovement)
      (Pawn|Components|CharacterMovement|AddImpulse _charactermovement (CallFunction|GetImpulseValue Pawn) true)
      (Variables|/Script/FortniteGame.FortPawn|SetbIgnoreNextFallingDamage true Pawn)
      (CallFunction|PlayKnockbackSound))))

(event Custom|KnockBackVehicle
  (params (VehicleToKnock :type Fort%20Athena%20Vehicle%20Object%20Reference) (ActorRef :type Actor%20Object%20Reference))
  ;[0 1 0.6223 0.25] KnockbackVehicles
  (Networking|SwitchHasAuthority
    (:Authority
      (Vehicle|WakeUp VehicleToKnock)
      (Components|Movement|StopMovementImmediately (Pawn|GetMovementComponent VehicleToKnock))
      (Environment|ApplyGlobalEnvironmentGameplayEffecttoActor VehicleToKnock (Variables|Default|GetDisableVehicleEffect))
      (if (Class|FortVehicleInterface|IsBoosting VehicleToKnock)
        (Physics|AddRadialForce (Variables|/Script/FortniteGame.FortAthenaVehicle|GetSkeletalMesh VehicleToKnock) (Transformation|GetWorldLocation (Variables|Default|GetKnockbackLocation)) 777.0 5e+01 "RIF_Constant" true) ; Apply Force to cancel out boost.
      )
      (Audio|BroadcastAbilityImpactatLocation 0 VehicleToKnock (CallFunction|GetImpulseValue ActorRef true) 777.0)
      (CallFunction|PlayKnockbackSound))))

(fn GetImpulseValue
  (params (OtherActor :type Actor%20Object%20Reference) (IsVehicle :type Boolean) (IsItem :type Boolean))
  (bind (_returnvalue_x _returnvalue_y) (Math|Vector|BreakVector (- (Transformation|GetActorLocation OtherActor) (Transformation|GetWorldLocation (Variables|Default|GetKnockbackLocation)))))
  ;[0 1 0.6054 0.25] Get impulse value based off Box location
  (Utilities|IsValid OtherActor
    (:Is%20Valid
      (return (+ (* (Math|Vector|Normalize (Math|Vector|MakeVector _returnvalue_x _returnvalue_y)) (Math|Vector|MakeVector (ScalableFloat|GetValueatLevel (select (LocalVariables|GetIsVehicle) (Variables|Default|GetImpulse%20Vehicle%20-%20Horizontal) (Variables|Default|GetImpulse%20Pawn%20-%20Horizontal)) 1.0) (ScalableFloat|GetValueatLevel (select (LocalVariables|GetIsVehicle) (Variables|Default|GetImpulse%20Vehicle%20-%20Horizontal) (Variables|Default|GetImpulse%20Pawn%20-%20Horizontal)) 1.0))) (Math|Vector|MakeVector 0.0 0.0 (* (ScalableFloat|GetValueatLevel (select (LocalVariables|GetIsVehicle) (Variables|Default|GetImpulse%20Vehicle%20-%20Verticle) (Variables|Default|GetImpulse%20Pawn%20-%20Verticle)) 1.0) 1.2)))))
    (:Is%20Not%20Valid
      (Development|PrintString "Fail to find other actor." true true "(R=0.000000,G=0.660000,B=1.000000,A=1.000000)" 2e+01))))

(fn ConstructionScript)
