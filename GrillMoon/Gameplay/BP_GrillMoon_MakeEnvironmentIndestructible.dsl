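; Grill Moon actor that makes nearby environment BuildingActors indestructible in Fortnite.
; On authority, it gathers BuildingActors in a box around itself and sets them invulnerable, undamageable and force-blocking building.
; The same duplicated pass runs on playlist data ready, the gm_propboost custom event and a one-time overlap of the player collision box, which then deactivates.
; A replicated prop count prints through its OnRep for debugging.
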(event OnComponentBeginOverlap(Box-Playercollision)
  (params (OverlappedComponent :type Primitive%20Component%20Object%20Reference) (OtherActor :type Actor%20Object%20Reference) (OtherComp :type Primitive%20Component%20Object%20Reference) (OtherBodyIndex :type Integer) (bFromSweep :type Boolean) (SweepResult :type Hit%20Result%20Structure%20%28by%20ref%29))
  (Utilities|FlowControl|DoOnce false
    (:Completed
      (Components|Activation|Deactivate (Variables|Default|GetBox%20-%20Player%20collision)) ; this doesn't actually disable it?
      (Networking|SwitchHasAuthority
        (:Authority
          (bind _outactors (Class|SpecialEventHelpers|GetActorsinBox (Utilities|Array|MakeArray "/Script/FortniteGame.BuildingActor") (Math|Box|MakeBox (- (Transformation|GetActorLocation) (Math|Conversions|ToVector(Float) 3e+03)) (+ (Transformation|GetActorLocation) (Math|Conversions|ToVector(Float) 3e+03)))))
          (Variables|Default|Set%28replicated%29%20total%20props (Utilities|Array|Length _outactors)) ; Replicate this down to clients purely\nfor logging. Trying to figure out why\nthis isn't working on launcher.
          (for _array_element _outactors
            (bind _asbuilding_actor (Utilities|Casting|CastToBuildingActor _array_element))
            (Variables|/Script/FortniteGame.BuildingActor|SetbIsInvulnerable true _asbuilding_actor)
            (Variables|/Script/FortniteGame.BuildingActor|SetbForceBlockBuildings true _asbuilding_actor)
            (Variables|/Script/Engine.Actor|SetbCanBeDamaged false _asbuilding_actor)))))) ; Do once when players get near,\nrun the logic again to disable \ndestructibility on the environment
)

(event Custom|gm_propboost
  (Networking|SwitchHasAuthority
    (:Authority
      (bind _outactors (Class|SpecialEventHelpers|GetActorsinBox (Utilities|Array|MakeArray "/Script/FortniteGame.BuildingActor") (Math|Box|MakeBox (- (Transformation|GetActorLocation) (Math|Conversions|ToVector(Float) 3e+03)) (+ (Transformation|GetActorLocation) (Math|Conversions|ToVector(Float) 3e+03)))))
      (Variables|Default|Set%28replicated%29%20total%20props (Utilities|Array|Length _outactors)) ; Replicate this down to clients purely\nfor logging. Trying to figure out why\nthis isn't working on launcher.
      (for _array_element _outactors
        (bind _asbuilding_actor (Utilities|Casting|CastToBuildingActor _array_element))
        (Variables|/Script/FortniteGame.BuildingActor|SetbIsInvulnerable true _asbuilding_actor)
        (Variables|/Script/FortniteGame.BuildingActor|SetbForceBlockBuildings true _asbuilding_actor)
        (Variables|/Script/Engine.Actor|SetbCanBeDamaged false _asbuilding_actor)))))

(event EventBeginPlay
  (Class|FortAsyncActionCurrentPlaylistReady|PlaylistDataReadyAsync
    (:OnReady
      (Networking|SwitchHasAuthority
        (:Authority
          (bind _outactors (Class|SpecialEventHelpers|GetActorsinBox (Utilities|Array|MakeArray "/Script/FortniteGame.BuildingActor") (Math|Box|MakeBox (- (Transformation|GetActorLocation) (Math|Conversions|ToVector(Float) 3e+03)) (+ (Transformation|GetActorLocation) (Math|Conversions|ToVector(Float) 3e+03)))))
          (Variables|Default|Set%28replicated%29%20total%20props (Utilities|Array|Length _outactors)) ; Replicate this down to clients purely\nfor logging. Trying to figure out why\nthis isn't working on launcher.
          (for _array_element _outactors
            (bind _asbuilding_actor (Utilities|Casting|CastToBuildingActor _array_element))
            (Variables|/Script/FortniteGame.BuildingActor|SetbIsInvulnerable true _asbuilding_actor)
            (Variables|/Script/FortniteGame.BuildingActor|SetbForceBlockBuildings true _asbuilding_actor)
            (Variables|/Script/Engine.Actor|SetbCanBeDamaged false _asbuilding_actor)))))))

(fn OnRep_%28replicated%29totalprops
  (Development|PrintString (Utilities|String|Append "GrillMoon - total props being made invulnerable: " (Utilities|String|ToString(Integer) (Variables|Default|Get%28replicated%29%20total%20props)))))

(fn ConstructionScript)
