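; Server-authoritative respawn region actor that allows respawn for Fortnite player pawns while they overlap it and are inside the safe zone.
; Begin and end overlap events pass the player's controller to the special-event respawn mutator, naming this region.
; A missing controller or mutator schedules a next-tick recheck of overlapping pawns.
; A looping timer also rechecks them every second, and EndPlay clears both timers.
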(event EventBeginPlay
  (Networking|SwitchHasAuthority
    (:Authority
      (CallFunction|CheckforOverlappingPawns)
      (bind _returnvalue (Utilities|Time|SetTimerbyEvent (EventDispatchers|CreateEvent :!event Custom|CheckForOverlappingPawns) 1.0 true))
      (Variables|Default|SetOverlapsCheckTimer _returnvalue))))

(event Collision|EventActorBeginOverlap
  (params (OtherActor :type Actor%20Object%20Reference))
  ;[0 1 0 0.5] [ActorBeginOverlap] Allow respawn for the overlapped player
  (CallFunction|SetPlayerRespawnAllowed OtherActor true))

(event Collision|EventActorEndOverlap
  (params (OtherActor :type Actor%20Object%20Reference))
  ;[1 1 0 0.5] [ActorEndOverlap] Stop allowing respawn for the overlapped player
  (CallFunction|SetPlayerRespawnAllowed OtherActor))

(event Custom|SetPlayerRespawnAllowed
  (params (Actor :type Actor%20Object%20Reference) (bAllowed :type Boolean))
  ;[0.25 0.1 0 0.5] [SetPlayerRespawnAllowed] Start/stop allowing respawn for a player
  (Networking|SwitchHasAuthority
    (:Authority
      (bind _asfort_player_pawn_athena (Utilities|Casting|CastToFortPlayerPawnAthena Actor))
      (bind _asfort_player_controller_zone (Utilities|Casting|CastToFortPlayerControllerZone (Pawn|GetController _asfort_player_pawn_athena))
        (:then
          (bind _returnvalue (Mutator|GetGameplayMutatorbyClass "/Script/SpecialEventGameplayRuntime.FortAthenaMutator_SE_Respawn"))
          (Utilities|IsValid _returnvalue
            (:Is%20Valid
              (Class|FortAthenaMutatorSERespawn|SetPlayerRespawnAllowed _returnvalue _asfort_player_controller_zone (and bAllowed (Variables|/Script/FortniteGame.FortPlayerPawn|GetbIsInsideSafeZone _asfort_player_pawn_athena)) self))
            (:Is%20Not%20Valid
              _)))
        (:CastFailed
          _))
      ;;[1 0 0 0.5] Retry next frame in a couple of cases:
      ;; 1. When a player respawns directly into one of our overlap components, we may get an OnBeginOverlap before the pawn has its controller pointer set.
      ;; 2. The mutator doesn't exist.
      (Utilities|Time|ClearandInvalidateTimerbyHandle (Variables|Default|GetCheckOverlapsTimerHandle))
      (bind _returnvalue_2 (Utilities|Time|SetTimerforNextTickbyEvent (EventDispatchers|CreateEvent :!event Custom|CheckForOverlappingPawns)))
      (Variables|Default|SetCheckOverlapsTimerHandle _returnvalue_2))))

(event Custom|CheckForOverlappingPawns
  ;[0 0 1 0.5] [CheckForOverlappingPawns] Retry all overlapping pawns, in case there was a prior failuires to enable respawn for any of them
  (for _array_element (Collision|GetOverlappingActors "/Script/FortniteGame.FortPlayerPawnAthena")
    (CallFunction|SetPlayerRespawnAllowed _array_element true)))

(event EventEndPlay
  (params (EndPlayReason :type EEndPlayReason%20Enum :default "Destroyed"))
  ;[0 0.25 0.25 0.5] [EndPlay] Clear timer handle
  (Utilities|Time|ClearandInvalidateTimerbyHandle (Variables|Default|GetCheckOverlapsTimerHandle))
  (Utilities|Time|ClearandInvalidateTimerbyHandle (Variables|Default|GetOverlapsCheckTimer)))

(fn ConstructionScript)
