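; Server-authoritative teleporter for the GrillMoon special event, driven by the stateful TeleportHover event.
; On activation it forces players out of hiding props, spawns HUD, map, focus and multi-squad relevancy mutators, then applies teleport effects to relevancy-grouped pawns, resetting teleport points per group.
; Deactivation removes the effects and restores normal relevancy.
; Debug events simulate bucketed teleports and draw or adjust the teleport zone box.
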(event Custom|TeleTestLoopAgain
  (bind _completedloopsvalue (Variables|Debug|GetCompletedLoopsValue))
  ; Debug: Used to test sending multiple "players" through the teleport system based on relevancy bucket size
  (Variables|Debug|SetTeleportsRemaining (- (Variables|Debug|GetTeleportsRemaining) (Variables|Debug|GetTeleportsThisLoop)))
  (Class|BPGMTeleportLocations|ResetPlayersTeleported (Variables|Default|GetTeleportLocations))
  (Utilities|FlowControl|ForLoopwithBreak 0 (Math|Integer|SelectInt (- (Variables|Debug|GetTeleportsRemaining) 1) (- (Math|Float|Truncate (Variables|Debug|GetPlayersPerBucket)) 1) (< (Variables|Debug|GetTeleportsRemaining) (Variables|Debug|GetPlayersPerBucket)))
    (:LoopBody
      (Variables|Debug|SetSelectedTeleportPoint (Class|BPGMTeleportLocations|GetUnoccupiedTeleportPoint (Variables|Default|GetTeleportLocations)))
      (Math|Integer|IncrementInt (Variables|Debug|GetTotal%20PlayersTeleported)
        (:%20%20
          (Math|Integer|IncrementInt (Variables|Debug|GetTeleportsThisLoop)))))
    (:Completed
      (Math|Integer|IncrementInt _completedloopsvalue
        (:%20%20
          (Variables|Debug|SetTeleportsThisLoop 0)
          (if (== _completedloopsvalue (Variables|Debug|GetRounded%20Loop%20value))
            (Development|PrintString (Utilities|String|Append "Teleported " (Utilities|String|ToString(Integer) (Variables|Debug|GetTotal%20PlayersTeleported)) " players with " (Utilities|String|ToString(Integer) (Variables|Debug|GetRounded%20Loop%20value)) " loops."))
            (else
              (CallFunction|TeleTestLoopAgain))))))))

(event Custom|GM_GetTeleActor
  (bind _returnvalue (Actor|GetActorOfClass "/GrillMoon/Gameplay/BP_GM_TeleportLocations.BP_GM_TeleportLocations_C"))
  (Variables|Default|SetTeleportLocations _returnvalue))

(event Custom|GM_TeleTest
  (params (TestNumber :type Integer))
  (bind _playersperbucket (Variables|Debug|GetPlayersPerBucket))
  (bind _totalteleports (Variables|Debug|GetTotalTeleports))
  (bind _returnvalue (/ _totalteleports _playersperbucket))
  (bind _returnvalue_2 (Math|Float|Round _returnvalue))
  (bind _completedloopsvalue (Variables|Debug|GetCompletedLoopsValue))
  ; Debug: Used to test sending multiple "players" through the teleport system based on relevancy bucket size
  (Variables|Debug|SetCompletedLoopsValue 0)
  (Variables|Debug|SetTotal%20PlayersTeleported 0)
  (Class|BPGMTeleportLocations|ResetPlayersTeleported (Variables|Default|GetTeleportLocations))
  (Variables|Debug|SetTotalTeleports TestNumber)
  (Variables|Debug|SetTeleportsRemaining TestNumber)
  (Variables|Debug|SetRounded%20Loop%20value _returnvalue_2)
  (if (> _returnvalue _returnvalue_2)
    (Math|Integer|IncrementInt (Variables|Debug|GetRounded%20Loop%20value)
      (:%20%20
        (Development|PrintString (Utilities|String|Append "Total loops needed " (Utilities|String|ToString(Integer) (Variables|Debug|GetRounded%20Loop%20value)) " for " (Utilities|String|ToString(Integer) (Variables|Debug|GetTotalTeleports)) "Players"))))
    (else
      _))
  (Utilities|FlowControl|ForLoopwithBreak 0 (Math|Integer|SelectInt (- (Variables|Debug|GetTeleportsRemaining) 1) (- (Math|Float|Truncate (Variables|Debug|GetPlayersPerBucket)) 1) (< (Variables|Debug|GetTeleportsRemaining) (Variables|Debug|GetPlayersPerBucket)))
    (:LoopBody
      (Variables|Debug|SetSelectedTeleportPoint (Class|BPGMTeleportLocations|GetUnoccupiedTeleportPoint (Variables|Default|GetTeleportLocations)))
      (Math|Integer|IncrementInt (Variables|Debug|GetTotal%20PlayersTeleported)
        (:%20%20
          (Math|Integer|IncrementInt (Variables|Debug|GetTeleportsThisLoop)))))
    (:Completed
      (Math|Integer|IncrementInt _completedloopsvalue
        (:%20%20
          (Variables|Debug|SetTeleportsThisLoop 0)
          (if (== _completedloopsvalue (Variables|Debug|GetRounded%20Loop%20value))
            (Development|PrintString (Utilities|String|Append "Teleported " (Utilities|String|ToString(Integer) (Variables|Debug|GetTotal%20PlayersTeleported)) " players with " (Utilities|String|ToString(Integer) (Variables|Debug|GetRounded%20Loop%20value)) " loops."))
            (else
              (CallFunction|TeleTestLoopAgain))))))))

(event EventBeginPlay
  (Networking|SwitchHasAuthority
    (:Authority
      (bind (_asynctaskproxy _payload) (GameplayEventRouter|StartListeningToStatefulEvent (GameplayEventRouter|GetGlobalEventRouter) "/GrillMoon/Gameplay/GameplayEvents/GrillMoon_TeleportHover_Active.GrillMoon_TeleportHover_Active")
        (:then
          (CallFunction|DebugSetTeleZoneSize))
        (:OnEventReceived
          (if (Utilities|Struct|BreakGrillMoonTeleportHoverActive _payload)
            (bind _returnvalue (Actor|GetActorOfClass "/GrillMoon/Gameplay/BP_GM_TeleportLocations.BP_GM_TeleportLocations_C"))
            (Variables|Default|SetTeleportLocations _returnvalue)
            (CallFunction|ShockwaveTeleport)
            (else
              (CallFunction|RemoveGMHover))))))))

(event Custom|GM_StopHide
  (Networking|SwitchHasAuthority
    (:Authority
      (CallFunction|ForceHidingStop))))

(event Custom|ShockwaveTeleport
  (bind _returnvalue (Class|SpecialEventHelpers|GetSpecialRelevancyMutator))
  (bind _grantabilitiesindex (Variables|Default|GetGrantAbilitiesIndex))
  (Networking|SwitchHasAuthority
    (:Authority
      (CallFunction|ForceHidingStop)
      (CallFunction|SpawnMutators true)
      (Variables|/GrillMoon/Gameplay/BP_GM_TeleportLocations.BP_GM_TeleportLocations_C|SetPlayersPerBucket (Variables|Default|GetMax%20Players%20Per%20Relevancy%20Group) (Variables|Default|GetTeleportLocations))
      (Utilities|IsValid _returnvalue
        (:Is%20Valid
          _)
        (:Is%20Not%20Valid
          _))
      (bind _relevantpawns (Class|FortAthenaMutatorSpecialRelevancy|GetRelevantPawnGroups _returnvalue))
      (for _array_element _relevantpawns
        (Class|BPGMTeleportLocations|ResetPlayersTeleported (Variables|Default|GetTeleportLocations))
        (Variables|Default|SetGrantAbilitiesIndex 0)
        (if (<= (Variables|Default|GetGrantAbilitiesIndex) (Utilities|Array|LastIndex (Utilities|Struct|BreakRelevantPawnArray _array_element)))
          (Utilities|IsValid (Ability|GetFortAbilitySystemComponent (Utilities|Array|Get(acopy) (Utilities|Struct|BreakRelevantPawnArray _array_element) _grantabilitiesindex))
            (:Is%20Valid
              (if (Pawn|IsPlayerControlled (Utilities|Array|Get(acopy) (Utilities|Struct|BreakRelevantPawnArray _array_element) _grantabilitiesindex))
                (GameplayEffects|ApplyGameplayEffectToTarget :self (Ability|GetFortAbilitySystemComponent (Utilities|Array|Get(acopy) (Utilities|Struct|BreakRelevantPawnArray _array_element) _grantabilitiesindex)) :GameplayEffectClass "/GrillMoon/Gameplay/Effects/GE_GrillMoon_Teleport.GE_GrillMoon_Teleport_C" :Target (Ability|GetFortAbilitySystemComponent (Utilities|Array|Get(acopy) (Utilities|Struct|BreakRelevantPawnArray _array_element) _grantabilitiesindex)) :Level 0.0)
                (GameplayEffects|ApplyGameplayEffectToTarget :self (Ability|GetFortAbilitySystemComponent (Utilities|Array|Get(acopy) (Utilities|Struct|BreakRelevantPawnArray _array_element) _grantabilitiesindex)) :GameplayEffectClass (Variables|Default|GetGameplay%20Effect%20Aura) :Target (Ability|GetFortAbilitySystemComponent (Utilities|Array|Get(acopy) (Utilities|Struct|BreakRelevantPawnArray _array_element) _grantabilitiesindex)) :Level 0.0)
                (bind _result (Math|Integer|IncrementInt (Variables|Default|GetGrantAbilitiesIndex))
                  (:%20%20
                    (if (== (Math|Integer|%25%28Integer%29 _result 25) 0)
                      (Utilities|FlowControl|Delay 0.025) ; delay for server perf
                    )))))) ; ability system component valid?
        )
        (Utilities|FlowControl|Delay 0.01) ; delay for server perf
      ))))

(event Custom|RemoveGMHover
  (bind _grantabilitiesindex (Variables|Default|GetGrantAbilitiesIndex))
  (Networking|SwitchHasAuthority
    (:Authority
      (CallFunction|SpawnMutators false)
      (bind _relevantpawns (Class|FortAthenaMutatorSpecialRelevancy|GetRelevantPawnGroups (Class|SpecialEventHelpers|GetSpecialRelevancyMutator)))
      (for _array_element _relevantpawns
        (Variables|Default|SetGrantAbilitiesIndex 0)
        (if (<= (Variables|Default|GetGrantAbilitiesIndex) (Utilities|Array|LastIndex (Utilities|Struct|BreakRelevantPawnArray _array_element)))
          (Utilities|IsValid (Ability|GetFortAbilitySystemComponent (Utilities|Array|Get(acopy) (Utilities|Struct|BreakRelevantPawnArray _array_element) _grantabilitiesindex))
            (:Is%20Valid
              (if (Pawn|IsPlayerControlled (Utilities|Array|Get(acopy) (Utilities|Struct|BreakRelevantPawnArray _array_element) _grantabilitiesindex))
                (GameplayEffects|RemoveActiveGameplayEffectbySourceEffect (Ability|GetFortAbilitySystemComponent (Utilities|Array|Get(acopy) (Utilities|Struct|BreakRelevantPawnArray _array_element) _grantabilitiesindex)) "/GrillMoon/Gameplay/Effects/GE_GrillMoon_Teleport.GE_GrillMoon_Teleport_C" 0 -1)
                (GameplayEffects|RemoveActiveGameplayEffectbySourceEffect (Ability|GetFortAbilitySystemComponent (Utilities|Array|Get(acopy) (Utilities|Struct|BreakRelevantPawnArray _array_element) _grantabilitiesindex)) (Variables|Default|GetGameplay%20Effect%20Aura) 0 -1)
                (bind _result (Math|Integer|IncrementInt (Variables|Default|GetGrantAbilitiesIndex))
                  (:%20%20
                    (if (== (Math|Integer|%25%28Integer%29 _result 25) 0)
                      (Utilities|FlowControl|Delay 0.025) ; delay for server perf
                    )))))) ; ability system component valid?
        ))
      (Class|FortAthenaMutatorSpecialRelevancy|StartNormalRelevancy (Class|SpecialEventHelpers|GetSpecialRelevancyMutator))
      (Mutator|SetMutatorActive (Variables|Default|GetRelevancyMutator) false))))

(event Custom|GM_DrawTeleZone
  ;[0 0.15 0.07697 0.5] Cheats to draw teleport extents/ modify location and extents
  (bind _returnvalue (Animation|Utilities|GetFacingRotation (Variables|Default|GetDebugTeleportCenterVector) (Variables|Default|GetPortalVector)))
  (Rendering|Debug|DrawDebugBox (Variables|Default|GetDebugTeleportCenterVector) (Variables|Default|GetDebugTeleportZoneExtent) "(R=0.026907,G=1.000000,B=0.000000,A=1.000000)" _returnvalue 3e+02 15.0))

(event Custom|GM_SetTeleZoneExtent
  (params (NewExtent :type Vector))
  ;[0 0.15 0.07697 0.5] Cheats to draw teleport extents/ modify location and extents
  (Variables|Default|SetDebugTeleportZoneExtent NewExtent))

(event Custom|GM_SetTeleCenterForDraw
  (params (NewVector :type Vector))
  ;[0 0.15 0.07697 0.5] Cheats to draw teleport extents/ modify location and extents
  (Variables|Default|SetDebugTeleportCenterVector NewVector))

(event Custom|GM_DefaultCenterForDraw
  ;[0 0.15 0.07697 0.5] Cheats to draw teleport extents/ modify location and extents
  (Variables|Default|SetDebugTeleportCenterVector (Variables|Default|GetTeleportCenterVectorDefault)))

(event Custom|GM_DefaultTeleExtent
  ;[0 0.15 0.07697 0.5] Cheats to draw teleport extents/ modify location and extents
  (Variables|Default|SetDebugTeleportZoneExtent (Variables|Default|GetTeleportExtentDefault)))

(fn DebugSetTeleZoneSize
  (bind _teleportboxheight (Variables|Default|GetTeleportBoxHeight))
  (bind _returnvalue (ScalableFloat|GetValueatLevelwithContext _teleportboxheight))
  (bind _teleportboxlength (Variables|Default|GetTeleportBoxLength))
  (bind _returnvalue_2 (ScalableFloat|GetValueatLevelwithContext _teleportboxlength))
  (bind _teleportboxdepth (Variables|Default|GetTeleportBoxDepth))
  (bind _returnvalue_3 (ScalableFloat|GetValueatLevelwithContext _teleportboxdepth))
  (bind _returnvalue_4 (Math|Vector|MakeVector _returnvalue_3 _returnvalue_2 _returnvalue))
  (Variables|Default|SetDebugTeleportZoneExtent _returnvalue_4)
  (Variables|Default|SetTeleportExtentDefault _returnvalue_4))

(fn ForceHidingStop
  (bind _outactors (GameplayTags|GetAllActorsOfClassMatchingTagQuery "/Game/Athena/Items/EnvironmentalItems/HidingProps/B_HidingProp.B_HidingProp_C" (Variables|Default|GetHidingTag)))
  (Variables|Default|SetPropsHidingPlayers _outactors)
  (for _array_element (Variables|Default|GetPropsHidingPlayers)
    (if (> (Utilities|Array|Length (Variables|/Game/Athena/Items/EnvironmentalItems/HidingProps/B_HidingProp.B_HidingProp_C|GetHidingPlayers _array_element)) 0)
      (for _array_element_2 (Variables|/Game/Athena/Items/EnvironmentalItems/HidingProps/B_HidingProp.B_HidingProp_C|GetHidingPlayers _array_element)
        (Class|BHidingProp|StopHiding _array_element _array_element_2)))))

(fn SpawnMutators
  (params (Activate :type Boolean))
  (if Activate
    (bind _returnvalue (Class|SpecialEventGameplayLibrary|SpawnMutator "/GrillMoon/Gameplay/M_SE_GrillMoon_BlockFullScreenMap.M_SE_GrillMoon_BlockFullScreenMap_C"))
    (Variables|Default|SetSE_BlockFullScreenMap _returnvalue)
    (bind _returnvalue_2 (Class|SpecialEventGameplayLibrary|SpawnMutator "/GrillMoon/Gameplay/M_SE_GrillMoon_BlockHUD.M_SE_GrillMoon_BlockHUD_C"))
    (Variables|Default|SetSE_BlockHUD _returnvalue_2)
    (bind _returnvalue_3 (Class|SpecialEventGameplayLibrary|SpawnMutator "/GrillMoon/Gameplay/M_SE_GrillMoon_LockonFocalPoint.M_SE_GrillMoon_LockonFocalPoint_C"))
    (Variables|Default|SetFocusMutator _returnvalue_3)
    (bind _returnvalue_4 (Class|SpecialEventGameplayLibrary|SpawnMutator "/GrillMoon/Gameplay/M_SE_GM_RelevancyMutator.M_SE_GM_RelevancyMutator_C"))
    (Variables|Default|SetRelevancyMutator _returnvalue_4)
    (Mutator|SetMutatorActive (Variables|Default|GetRelevancyMutator) true)
    (Class|FortAthenaMutatorSpecialRelevancy|StartMultiSquadRelevancy (Class|SpecialEventHelpers|GetSpecialRelevancyMutator) (Variables|Default|GetMax%20Players%20Per%20Relevancy%20Group))
    (else
      (Utilities|FlowControl|Sequence
        (:then_0
          (bind _returnvalue_5 (Mutator|GetGameplayMutatorbyClass "/GrillMoon/Gameplay/M_SE_GrillMoon_BlockFullScreenMap.M_SE_GrillMoon_BlockFullScreenMap_C")))
        (:then_1
          (bind _returnvalue_6 (Mutator|GetGameplayMutatorbyClass "/GrillMoon/Gameplay/M_SE_GrillMoon_BlockHUD.M_SE_GrillMoon_BlockHUD_C")))
        (:then_2
          (bind _returnvalue_7 (Mutator|GetGameplayMutatorbyClass "/GrillMoon/Gameplay/M_SE_GrillMoon_LockonFocalPoint.M_SE_GrillMoon_LockonFocalPoint_C"))))
      (Mutator|SetMutatorActive _returnvalue_5 false))))

(fn ConstructionScript)
