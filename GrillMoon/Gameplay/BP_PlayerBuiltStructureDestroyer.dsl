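; Fortnite actor Blueprint that destroys player-built structures within its area when its custom Start Destroying event is called.
; The event runs only with authority and is gated by a ScalableFloat setting above zero.
; A box overlap uses the actor's location, with its scale as extent, to find BuildingActors, then destroys player-built ones and prints debug text.
; The fall-damage avoidance named in comments is an empty Sequence branch.
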(event Custom|Start%20Destroying%20Player%20Built%20Structures
  ;[1 1 1] Destroy Player Built Structures & Avoid Fall Damage
  (Networking|SwitchHasAuthority
    (:Authority
      (Utilities|FlowControl|Sequence
        (:then_0
          ;;[1 1 1] Destroy Player Built Structures
          (if (> (ScalableFloat|GetValueatLevel (Variables|PlayerBuiltStructureDestroyer|GetShould%20Break%20Player%20Built%20Structures) 1.0) 0)
            (bind _outactors (Collision|BoxOverlapActors (Transformation|GetActorLocation self) (Transformation|GetActorScale3D self) (Utilities|Array|MakeArray "ObjectTypeQuery1") "/Script/FortniteGame.BuildingActor" :!class /Script/Engine.KismetSystemLibrary)) ; find player built structures
            (for _array_element _outactors
              (bind _asbuilding_actor (Utilities|Casting|CastToBuildingActor _array_element))
              (if (BuildingActor|IsPlayerBuilt _asbuilding_actor)
                (Actor|DestroyActor _asbuilding_actor) ; Destroy the player made building actor
                (Development|PrintText (Utilities|Text|FormatText "NSLOCTEXT(\"[09DD528002FEB98EE1D967DE24720684]\", \"96A0CCD54C13D54830D770BF13A04E2D\", \"DESTROY  player built: {building}\")" (Utilities|GetDisplayName _asbuilding_actor)) false)))))
        (:then_1)) ; 0 - destroy player built buildings!\n1 - Player Fall Damage
    )))

(fn ConstructionScript)
