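; Gold Junk Rift spawner that opens a portal near players and drops a main actor plus randomized gold currency extras.
; On authority, an active calendar event and chance roll, or a GM event, starts timed proximity checks filtered by the junk manager's per-player rift count.
; Activation happens once, chaining delayed effects, server-side drops scattered around the portal, then shrink cleanup.
; Resource type and mesh index are chosen at spawn to avoid client-server desync.
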(event EventEndPlay
  (params (EndPlayReason :type EEndPlayReason%20Enum :default "Destroyed"))
  (|CallParentFunction EndPlayReason)
  (Utilities|Time|ClearandInvalidateTimerbyHandle (Variables|Default|GetPlayerCheckTimerHandle)))

(event Custom|GM_ActivateAllJunkRifts
  (if (Utilities|Time|IsValidTimerHandle (Variables|Default|GetPlayerCheckTimerHandle))
    _
    (elif (Variables|Default|GetZoneControlled)
      _
      (else
        (CallFunction|StartRiftTimer)))))

(event Custom|GM_ShowRiftIndicator
  (Variables|Default|SetShowIndicator (not (Rendering|IsVisible (Variables|Default|GetLocationIndicator)))))

(event EventBeginPlay
  (|CallParentFunction)
  (Networking|SwitchHasAuthority
    (:Authority
      (Class|FortAsyncActionAuthorityMatchReady|MatchReadyAsync
        (:OnReady
          (MCPContext|InitialCalendarSyncCompleted
            (:OnSyncComplete
              (MCPContext|WatchCalendarEvent (Variables|Default|GetGoldEventCalendarEventName)
                (:OnEventActive
                  (if (< (Math|Random|RandomIntegerinRange 0 100) (Variables|Default|GetChanceToActivate))
                    (if (not (Variables|Default|GetZoneControlled))
                      (CallFunction|StartRiftTimer))))))))))))

(event Custom|StartRiftTimer
  (if (not (Utilities|Time|IsValidTimerHandle (Variables|Default|GetPlayerCheckTimerHandle)))
    (bind _returnvalue (Utilities|Time|SetTimerbyEvent (EventDispatchers|CreateEvent :!event Custom|CheckForNearbyPlayers) (Variables|Default|GetPlayerCheckTime) true))
    (Variables|Default|SetPlayerCheckTimerHandle _returnvalue)))

(event Custom|SpawnStartSFX
  (bind _returnvalue (Niagara|SpawnSystematLocation "/GrillMoon/Assets/FX/GoldJunkRift/NS_GrillMoon_Jolly_Launch.NS_GrillMoon_Jolly_Launch" (Transformation|GetActorLocation)))
  (Utilities|IsValid _returnvalue
    (:Is%20Valid
      (Effects|Components|ParticleSystem|SetVectorParameter _returnvalue "Scale" (2.0 1.0 1.0)) ; Extend lightning to portal spawn
      (Networking|ForceNetUpdate))))

(event Custom|SpawnMainDrop
  ;[1 1 1] Drop the main actor
  (Game|SpawnBuildingActorfromClass (Variables|Default|GetDropperClassToSpawn) (Math|Transform|MakeTransform (Transformation|GetActorLocation) (Math|Rotator|MakeRotator 0.0 0.0 (Math|Random|RandomFloatinRange 0.0 3.6e+02)) "1.000000,1.000000,1.000000") (Game|GetInstigator) false (Game|GetInstigator))
  (Utilities|FlowControl|Delay (Variables|Default|GetDelay_AfterMainDrop))
  (CallFunction|SpawnExtraDrops))

(event Custom|SpawnPortalSFX
  (bind _riftfxniagara (Variables|Default|GetRiftFXNiagara))
  (Components|Activation|Activate (Variables|Default|GetAmbientAudo) true)
  (Components|Activation|Activate _riftfxniagara true)
  (Effects|Components|ParticleSystem|SetFloatParameter _riftfxniagara "Lifetime" (* (Variables|Default|GetDelayBeforeDrop) 4.0)))

(event Custom|SpawnExtraDrops
  (bind _returnvalue (Game|GetInstigator))
  (bind _spawnedextracount (Variables|Default|GetSpawnedExtraCount))
  (bind _actorstodrop (Variables|Default|GetActorsToDrop))
  (bind _output (Utilities|Array|Get(acopy) _actorstodrop _spawnedextracount))
  ;[1 1 1] Loop through and drop the array of actors chosen.
  (Utilities|FlowControl|Delay (Variables|Default|GetDelay_BeforeExtraDrops))
  (Variables|Default|SetSpawnedExtraCount 0)
  (if (not (>= (Variables|Default|GetSpawnedExtraCount) (Utilities|Array|Length (Variables|Default|GetActorsToDrop))))
    (bind _output_get (Variables|Default|SetExtraSpawnTemp (CallFunction|MakeExtraSpawnerTransform)))
    (Ability|ExecuteGameplayCueWithParamsReplicated (Variables|Cues|GetGC_SpawnSecondary) (Ability|GameplayCue|MakeGameplayCueParameters :NormalizedMagnitude 0.0 :RawMagnitude 0.0 :Location (.location _output_get) :Normal "0,0,0" :GameplayEffectLevel 1 :AbilityLevel 1 :bReplicateLocationWhenUsingMinimalRepProxy false))
    (Utilities|Casting|CastToBGA_Athena_JollyRascal_ExtraDrop_Resource_GMClass _output
      (:then
        (CallFunction|SpawnResourceBox _output_get _returnvalue)
        (Math|Integer|IncrementInt (Variables|Default|GetSpawnedExtraCount)
          (:%20%20
            (Utilities|FlowControl|Delay (Math|Random|RandomFloatinRange (Variables|Default|GetDelay_ExtraSpawnMin) (Variables|Default|GetDelay_ExtraSpawnMax))))))
      (:CastFailed
        _))
    (Game|SpawnBuildingActorfromClass _output _output_get _returnvalue false _returnvalue))
  ;[1 1 1] Clean up and kill after drop
  (Utilities|FlowControl|Delay (Variables|Default|GetDelay_AfterDropsShrink))
  (CallFunction|CleanUpandKill))

(event Custom|SpawnInPortal
  ;[1 1 1] Spawn In Portal - play FX and scale up portal.
  (CallFunction|SpawnPortalSFX)
  (Ability|ExecuteGameplayCueReplicated (Variables|Cues|GetGC_Start))
  (Utilities|FlowControl|Delay (Variables|Default|GetDelayBeforeDrop))
  ;[1 1 1] Once the portal has scaled up do the drop
  (if (Networking|IsServer)
    (CallFunction|SpawnMainDrop)))

(event Custom|ActivatePortal
  ;[1 1 1] Spawn Portal
  (if (not (Variables|Default|GetHasSpawned))
    (Variables|Default|SetHasSpawned true)
    (Utilities|Time|ClearandInvalidateTimerbyHandle (Variables|Default|GetPlayerCheckTimerHandle))
    (Rendering|SetActorHiddenInGame false)
    (CallFunction|SpawnStartSFX)
    (Utilities|FlowControl|Delay (Variables|Default|GetDelayBeforePortal))
    (CallFunction|SpawninPortal)
    (if (Networking|IsServer)
      (CallFunction|SelectThingstoDrop))))

(event Custom|CleanUpAndKill
  ;[1 1 1] Clean Up And Kill - Shrink Down the portal, play FX, hide, and set to kill.
  (Components|Activation|Deactivate (Variables|Default|GetAmbientAudo))
  (Ability|ExecuteGameplayCueReplicated (Variables|Cues|GetGC_End))
  (bind (_direction _alpha) (Timeline|ScaleDownPortal PlayFromStart)
    (:Update
      (Transformation|SetRelativeScale3D (Variables|Default|GetMesh) (Math|Vector|Lerp(Vector) (Variables|Default|GetPortalLargeScale) (Variables|Default|GetPortalSmallScale) _alpha)))
    (:Finished
      (Rendering|SetActorHiddenInGame false))))

(event Custom|CheckForNearbyPlayers
  ;[0.002363 0.15 0 0.5] If not controlled by the event zone, have this spawner do it's own range checks to see if a player is near by.
  (Variables|Default|SetPlayersInRange 0.0)
  (for _array_element (Variables|/Script/Engine.GameStateBase|GetPlayerArray (Game|GetGameState))
    (if (CallFunction|IsAValidPlayerWithinRange _array_element)
      (if (not (Class|BPGMJRManager|CheckPlayerRiftCount (Variables|Default|GetJunkManager) _array_element))
        (Math|Float|IncrementFloat (Variables|Default|GetPlayersInRange)))))
  (if (> (Variables|Default|GetPlayersInRange) 0)
    (CallFunction|ActivatePortal)))

(fn IsAValidPlayerWithinRange
  (params (PlayerState :type Player%20State%20Object%20Reference))
  (bind _returnvalue (PlayerState|GetPawn PlayerState))
  (bind _self self)
  (bind _returnvalue_2 (Transformation|GetDistanceTo _returnvalue _self))
  (Utilities|IsValid PlayerState
    (:Is%20Valid
      (Utilities|IsValid _returnvalue
        (:Is%20Valid
          (if (Pawn|IsPlayerControlled _returnvalue)
            (Development|PrintString (Utilities|String|Append "Nearest Player is " (Utilities|String|ToString(Float) _returnvalue_2) " away.") true true "(R=0.000000,G=0.660000,B=1.000000,A=1.000000)" 3e+01)
            (if (<= _returnvalue_2 (ScalableFloat|GetValueatLevelwithContext (Variables|Default|GetMaxCheckDistance)))
              (return true)
              (else
                _))
            (else
              _))
          (return false))
        (:Is%20Not%20Valid
          _)))
    (:Is%20Not%20Valid
      _)))

(fn MakeExtraSpawnerTransform
  (bind (_returnvalue_x _returnvalue_y) (Math|Vector|BreakVector (* (Math|Rotator|GetRotationXVector (Math|Rotator|MakeRotator 0.0 0.0 (Math|Random|RandomFloatinRange 0.0 3.6e+02))) (LocalVariables|GetExtraDropSpawnRadius))))
  (bind (_returnvalue_x_2 _returnvalue_y_2 _returnvalue_z) (Math|Vector|BreakVector (Transformation|GetActorLocation)))
  (return (Math|Transform|MakeTransform (Math|Vector|MakeVector (+ _returnvalue_x_2 _returnvalue_x) (+ _returnvalue_y_2 _returnvalue_y) _returnvalue_z) (Math|Rotator|MakeRotator 0.0 0.0 (Math|Random|RandomFloatinRange 0.0 3.6e+02)) "1.000000,1.000000,1.000000")))

(fn OnRep_ShowIndicator
  (Networking|CanExecuteCosmeticEvents
    (:True
      (Rendering|SetVisibility (Variables|Default|GetLocationIndicator) (Variables|Default|GetShowIndicator) false))))

(fn SelectThingsToDrop
  (bind _actorstodrop (Variables|Default|GetActorsToDrop))
  (Utilities|Array|Clear _actorstodrop)
  (for _ (range 1 (+ (Math|Random|RandomIntegerinRange (Math|Float|Truncate (ScalableFloat|GetValueatLevel (Variables|Rows|GetRow_MinExtraJunkToDrop))) (Math|Float|Truncate (ScalableFloat|GetValueatLevel (Variables|Rows|GetRow_MaxExtraJunkToDrop)))) 1))
    (Utilities|Array|Add _actorstodrop (LocalVariables|GetResourceBox)))
  (return))

(fn SpawnResourceBox
  (params (Transform :type Transform) (Instigator :type Pawn%20Object%20Reference))
  ;[1 1 1] Select the resource and spawn the actor. We do this here rather than in the spawned actor so that the server is spawning the actor with the resource type it is as to not have client server desync. We select both the resource type and a random mesh index to use on the actor.
  (LocalVariables|SetResourceType "GoldCurrency")
  (Game|SpawnBuildingActorfromClass "/GrillMoon/Gameplay/BGA_Athena_JollyRascal_ExtraDrop_Resource_GM.BGA_Athena_JollyRascal_ExtraDrop_Resource_GM_C" Transform Instigator false (LocalVariables|GetResourceType) (Math|Random|RandomIntegerinRange 0 1) 0 Instigator))

(fn ConstructionScript)
