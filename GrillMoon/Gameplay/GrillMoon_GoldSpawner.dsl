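; GrillMoon gold spawner actor that drops a random-sized Wads pickup at its location after a delay.
; Picking it up restarts the timer, with delay and amount bounds read from scalable floats.
; On authority, BeginPlay starts the timer from a watched calendar event or a specific Day 11 stateful router state, guarding against duplicate timers.
; Custom events activate or deactivate it; deactivation and EndPlay clear the timer and pickup binding.
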(event Custom|StartRespawn
  (params (SelfActor :type Fort%20Pickup%20Object%20Reference) (InteractingPawn :type Fort%20Pawn%20Object%20Reference))
  (Interaction|UnbindEventfromOnPickup (Variables|Default|GetSpawnedPickup) (delegate Custom|StartRespawn))
  (Development|PrintString "Gathering Dust")
  (bind _returnvalue (Utilities|Time|SetTimerbyEvent (EventDispatchers|CreateEvent :!event Custom|SpawnWads) (ScalableFloat|GetValueatLevelwithContext (Variables|Default|GetSpawnTimeLength))))
  (Variables|Default|SetSpawnTimer _returnvalue))

(event Custom|GrillMoon_ActivateGoldSpawner
  (bind _returnvalue (Utilities|Time|SetTimerbyEvent (EventDispatchers|CreateEvent :!event Custom|SpawnWads) (ScalableFloat|GetValueatLevelwithContext (Variables|Default|GetSpawnTimeLength))))
  (Variables|Default|SetSpawnTimer _returnvalue))

(event Custom|ActivateSpawnersViaStateEvent
  (if (not (Utilities|Time|IsValidTimerHandle (Variables|Default|GetSpawnTimer)))
    (bind _returnvalue (Utilities|Time|SetTimerbyEvent (EventDispatchers|CreateEvent :!event Custom|SpawnWads) (ScalableFloat|GetValueatLevelwithContext (Variables|Default|GetSpawnTimeLength))))
    (Variables|Default|SetSpawnTimer _returnvalue)))

(event EventBeginPlay
  (Networking|SwitchHasAuthority
    (:Authority
      (MCPContext|WatchCalendarEvent (Variables|Default|GetCalendarEventToWatch)
        (:then
          (bind (_asynctaskproxy _payload) (GameplayEventRouter|StartListeningToStatefulEvent (GameplayEventRouter|GetGlobalEventRouter) "/GrillMoon/Gameplay/EventRouter/GrillMoon_Day11_Events.GrillMoon_Day11_Events")
            (:OnEventReceived
              _)
            (:OnSavedEventStateExists
              _))
          (if (Utilities|Enum|Equal(Enum) (Utilities|Struct|BreakGrillMoonDay11Events _payload) "NewEnumerator3")
            (CallFunction|ActivateSpawnersViaStateEvent)))
        (:OnEventActive
          (if (not (Utilities|Time|IsValidTimerHandle (Variables|Default|GetSpawnTimer)))
            (bind _returnvalue (Utilities|Time|SetTimerbyEvent (EventDispatchers|CreateEvent :!event Custom|SpawnWads) (ScalableFloat|GetValueatLevelwithContext (Variables|Default|GetSpawnTimeLength))))
            (Variables|Default|SetSpawnTimer _returnvalue)))))))

(event Custom|SpawnWads
  (bind _returnvalue (Inventory|SpawnPickupInWorld "/Game/Items/ResourcePickups/Athena_WadsItemData.Athena_WadsItemData" (Math|Random|RandomIntegerinRange (ScalableFloat|GetValueasIntegerwithContext (Variables|Default|GetMinWads)) (ScalableFloat|GetValueasIntegerwithContext (Variables|Default|GetMaxWads))) (Transformation|GetActorLocation self)))
  (Variables|Default|SetSpawnedPickup _returnvalue)
  (Utilities|Time|ClearandInvalidateTimerbyHandle (Variables|Default|GetSpawnTimer))
  (Interaction|BindEventtoOnPickup (Variables|Default|GetSpawnedPickup) (EventDispatchers|CreateEvent :!event Custom|StartRespawn)))

(event Custom|GrillMoon_DeactivateGoldSpawner
  (Utilities|Time|ClearandInvalidateTimerbyHandle (Variables|Default|GetSpawnTimer))
  (Interaction|UnbindEventfromOnPickup (Variables|Default|GetSpawnedPickup) (EventDispatchers|CreateEvent :!event Custom|StartRespawn)))

(event EventEndPlay
  (params (EndPlayReason :type EEndPlayReason%20Enum :default "Destroyed"))
  (Networking|SwitchHasAuthority
    (:Authority
      (Utilities|Time|ClearandInvalidateTimerbyHandle (Variables|Default|GetSpawnTimer))
      (Interaction|UnbindEventfromOnPickup (Variables|Default|GetSpawnedPickup) (EventDispatchers|CreateEvent :!event Custom|StartRespawn)))))

(fn ConstructionScript)
